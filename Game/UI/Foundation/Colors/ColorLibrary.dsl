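; A UI color function library that reads named colors from a Color Stylesheet struct for stat and item display.
; GetBase&BuffColors breaks the stylesheet, then selects a base color by stat display type (fixed dark for Basic, themed for Unique and elemental) and a buff color by buff state, transparent when unchanged.
; The other functions are getters that break the stylesheet for buff, debuff, unique, elemental and harvest weak-point colors.
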(fn GetBase&BuffColors
  (params (Display%20Type :type EFortStatValueDisplayType%20Enum :default "BasicPaired") (Buff%20State :type EFortBuffState%20Enum :default "NoChange") (Stylesheet :type Color%20Stylesheet%20Structure))
  (bind (_theme_default_91_015cb18f413b4a467b38cb93e8734a68 _theme_emphasis_94_962f11a944e9df11f547fbbdb98dd4be _theme_critical_97_43f63d0c4e71395016ae00b41bb7ba0c _color_buff_145_ac063fb242d9ecbd0b6306bbe205c7f7 _color_debuff_146_0402338944fcc3ded4c90e9a78fd8bc5 _color_unique_150_b15cf0c9436988001f70b4b8f9d32e5b _color_fireelemental_152_fbafda0c445a96f1b689e880a707069c _color_iceelemental_154_fda209d8429df768aba53eba24c4b505 _color_boltelemental_156_10aefec147c25b2dee47619e4ab4abc7) (Utilities|Struct|BreakColorStylesheet Stylesheet))
  (return (Utilities|Select :Index Display%20Type :BasicPaired "(R=0.008023,G=0.010960,B=0.026241,A=1.000000)" :BasicSingle "(R=0.008023,G=0.010960,B=0.026241,A=1.000000)" :Unique _color_unique_150_b15cf0c9436988001f70b4b8f9d32e5b :ElementalFire _color_fireelemental_152_fbafda0c445a96f1b689e880a707069c :ElementalIce _color_iceelemental_154_fda209d8429df768aba53eba24c4b505 :ElementalElectric _color_boltelemental_156_10aefec147c25b2dee47619e4ab4abc7) (Utilities|Select :Index Buff%20State :NoChange "(R=0.000000,G=0.000000,B=0.000000,A=0.000000)" :Better _color_buff_145_ac063fb242d9ecbd0b6306bbe205c7f7 :Worse _color_debuff_146_0402338944fcc3ded4c90e9a78fd8bc5)))

(fn GetBoltElementalColor
  (params (Stylesheet :type Color%20Stylesheet%20Structure))
  (return (Utilities|Struct|BreakColorStylesheet Stylesheet)))

(fn GetBuffColor
  (params (Stylesheet :type Color%20Stylesheet%20Structure))
  (return (Utilities|Struct|BreakColorStylesheet Stylesheet)))

(fn GetDebuffColor
  (params (Stylesheet :type Color%20Stylesheet%20Structure))
  (return (Utilities|Struct|BreakColorStylesheet Stylesheet)))

(fn GetFireElementalColor
  (params (Stylesheet :type Color%20Stylesheet%20Structure))
  (return (Utilities|Struct|BreakColorStylesheet Stylesheet)))

(fn GetHarvestWeakPointColor
  (params (StyleSheet :type Color%20Stylesheet%20Structure))
  (return (Utilities|Struct|BreakColorStylesheet StyleSheet)))

(fn GetIceElementalColor
  (params (Stylesheet :type Color%20Stylesheet%20Structure))
  (return (Utilities|Struct|BreakColorStylesheet Stylesheet)))

(fn GetUniqueColor
  (params (Stylesheet :type Color%20Stylesheet%20Structure))
  (return (Utilities|Struct|BreakColorStylesheet Stylesheet)))
